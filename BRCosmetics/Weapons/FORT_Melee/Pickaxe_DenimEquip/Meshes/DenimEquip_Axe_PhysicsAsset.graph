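# Physics asset for the DenimEquip two-handed pickaxe, modeling its handle and a star pendant on a short chain.
# Capsule bodies cover the main handle, sword handle, chain base, two dynamic chain links and pendant.
# Constraints join them in series, with free angular motion from handle through chain base.
# The dyn_ links and pendant carry angular limits, with zero twist on the links and a looser pendant joint for dangling motion.
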PhysicsAsset /BRCosmetics/Weapons/FORT_Melee/Pickaxe_DenimEquip/Meshes/DenimEquip_Axe_PhysicsAsset
bodies {
 Melee_TwoHanded_Handle [{
   shapeType=Capsule
   length=100
   radius=10
   shapeName=Melee_TwoHanded_Handle_capsule
   center=<0.9214 -1.74 27.603>
   rotation=<rot 1.8813 -90 -0.2252>
  }]
 base_starchain [{shapeType=Capsule, length=0.505, radius=0.505, shapeName=base_starchain_capsule, rotation=<rot 0 0 90>}]
 dyn_starchain [{shapeType=Capsule, length=1, radius=0.6, shapeName=dyn_starchain_capsule, center=<2 0 0>, rotation=<rot 90 0 0>}]
 dyn_starchain_01 [{shapeType=Capsule, length=1, radius=0.6, shapeName=dyn_starchain_01_capsule, center=<2.2 0 0>, rotation=<rot 90 0 0>}]
 dyn_starpendant [{shapeType=Capsule, length=5.3168, radius=5.0474, shapeName=dyn_starpendant_capsule, center=<5.5 0 0>, rotation=<rot 0 0 90>}]
 swordhandle [{shapeType=Capsule, length=30, radius=8, shapeName=swordhandle_capsule, center=<-3 0 0>, rotation=<rot 0 0 90>}]
}
constraints [
 {
  bone1Name=swordhandle
  bone2Name=Melee_TwoHanded_Handle
  swing1LimitDegrees=45
  swing1Motion=Free
  swing2LimitDegrees=45
  swing2Motion=Free
  twistLimitDegrees=45
  twistMotion=Free
 }
 {
  bone1Name=base_starchain
  bone2Name=swordhandle
  swing1LimitDegrees=45
  swing1Motion=Free
  swing2LimitDegrees=45
  swing2Motion=Free
  twistLimitDegrees=45
  twistMotion=Free
 }
 {bone1Name=dyn_starchain, bone2Name=base_starchain, swing1LimitDegrees=35, swing2LimitDegrees=35, twistLimitDegrees=0}
 {bone1Name=dyn_starchain_01, bone2Name=dyn_starchain, swing1LimitDegrees=35, swing2LimitDegrees=35, twistLimitDegrees=0}
 {bone1Name=dyn_starpendant, bone2Name=dyn_starchain_01, swing1LimitDegrees=45, swing2LimitDegrees=45, twistLimitDegrees=45}
]
